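# Unlit masked Niagara sprite material for the rock particles of the Meteors vehicle boost.
# Blends two rock flipbook textures by dynamic parameter, with optional time-driven sub-UV stepping (off by default) and alpha scaled by particle alpha.
# Fakes lighting by dotting transformed texture RGB with +Z to mix inverted-grayscale shading terms, adding panned-noise, particle-tinted highlights, into MF_FX_EmissiveAdjustments.
# A near-duplicate normal-sampler branch has no visible consumer.
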Material /VehicleCosmetics/Boosts/Boost_Meteors/Material/M_Boost_Meteors
graph {
 properties {
  bAutomaticallySetUsageInEditor=false
  bUsedWithNiagaraSprites=true
  blendMode=BLEND_Masked
  opacityMaskClipValue=0.85
  refractionMethod=RM_2DOffset
  shadingModel=MSM_Unlit
  usedShadingModels=Unlit
 }
 nodes [
  MaterialExpressionVertexColor_4 = ParticleColor()
  n1 = TextureSampleParameter2D() {
   parameterName=FlipBook_01_Alpha
   samplerType=SAMPLERTYPE_LinearColor
   texture=/DealershipSharedLibrary/FX/Textures/Flipbook/T_VCVFX_Rocksv4_N
  }
  n2 = TextureSampleParameter2D() {
   parameterName=Flipbook_02_Alpha
   samplerType=SAMPLERTYPE_LinearColor
   texture=/DealershipSharedLibrary/FX/Textures/Flipbook/T_VCVFX_Rocksv3_N
  }
  n3 = DynamicParameter() {paramNames [TextureBlend,ColorPower,SpinSpeed,Brightness]}
  n4 = LinearInterpolate(A=@n1.A, B=@n2.A, Alpha=@n3.TextureBlend)
  MaterialExpressionConstantClamp_12 = Saturate(None=Multiply(A=@MaterialExpressionVertexColor_4.A, B=StaticSwitchParameter(True=@n4, False=@n1.A) {defaultValue=true, parameterName=Blend}))
  n5 = TextureCoordinate()
  n6 = ScalarParameter() {defaultValue=4, group=FlipBook, parameterName=Columns}
  n7 = ScalarParameter() {defaultValue=4, group=FlipBook, parameterName=Rows}
  n8 = AppendVector(A=@n6, B=@n7)
  n9 = Multiply(A=@n6, B=@n7)
  n10 = DynamicParameter() {paramNames [TextureBlend,ColorPower,SpinSpeed,Brightness]}
  n11 = Frac(None=Multiply(A=Time(), B=@n10.SpinSpeed))
  n12 = Floor(None=Multiply(A=AppendVector(A=@n9, B=@n7), B=AppendVector(A=@n11, B=@n11)))
  n13 = StaticSwitchParameter(
   True=Add(A=Divide(A=@n5, B=@n8), B=Divide(A=@n12, B=@n8))
   False=Multiply(A=TextureCoordinate(), B=ScalarParameter() {defaultValue=1, group=FlipBook, parameterName=TileAmount})
  ) {parameterName=SubMovie?}
  n14 = TextureSampleParameter2D(UVs=@n13) {parameterName=FlipBook_01, samplerType=SAMPLERTYPE_Normal, texture=&T_VCVFX_Rocksv4_N}
  n15 = TextureSampleParameter2D(UVs=@n13) {parameterName=Flipbook_02, samplerType=SAMPLERTYPE_Normal, texture=&T_VCVFX_Rocksv3_N}
  n16 = StaticSwitchParameter(True=LinearInterpolate(A=@n14.RGB, B=@n15.RGB, Alpha=@n10.TextureBlend), False=@n14.RGB) {defaultValue=true, parameterName=Blend}
  n17 = OneMinus(None=Desaturation(None=@n16))
  n18 = Saturate(None=@n17)
  n19 = Panner(Coordinate=TextureCoordinate() {uTiling=0.05, vTiling=0.05}) {speedX=0.25, speedY=0.25}
  n20 = Add(
   A=Multiply(A=TextureSample(UVs=@n19) {texture=/DealershipSharedLibrary/FX/Textures/TexturePackage/T_VCVFX_NoiseSpots_RGB}.RGB, B=Constant() {r=100})
   B=ParticleColor().RGB
  )
  n21 = Constant() {r=1.25}
  n22 = Saturate(None=Multiply(A=Power(Base=@n20, Exp=@n21), B=ScalarParameter() {defaultValue=5, parameterName=ReflectionAmount}))
  n23 = ParticleColor()
  n24 = ConstantBiasScale(None=@n17) {bias=-0.2, scale=0.2}
  n25 = Constant3Vector() {constant=<rgba 0 0 1 0>}
  n26 = Saturate(None=DotProduct(A=Transform(None=@n16), B=Normalize(VectorInput=@n25)))
  LinearInterpolate(
   A=Add(A=Power(Base=@n18, Exp=@n10.ColorPower), B=Multiply(A=@n22, B=@n23.RGB))
   B=Multiply(A=Saturate(None=@n24), B=@n10.Brightness)
   Alpha=Saturate(None=ConstantBiasScale(None=@n26) {bias=0.05})
  ) {constA=1}
  n27 = TextureCoordinate()
  n28 = ScalarParameter() {defaultValue=4, group=FlipBook, parameterName=Columns}
  n29 = ScalarParameter() {defaultValue=4, group=FlipBook, parameterName=Rows}
  n30 = AppendVector(A=@n28, B=@n29)
  n31 = Multiply(A=@n28, B=@n29)
  n32 = Frac(None=Multiply(A=Time(), B=@n3.SpinSpeed))
  n33 = Floor(None=Multiply(A=AppendVector(A=@n31, B=@n29), B=AppendVector(A=@n32, B=@n32)))
  n34 = StaticSwitchParameter(
   True=Add(A=Divide(A=@n27, B=@n30), B=Divide(A=@n33, B=@n30))
   False=Multiply(A=TextureCoordinate(), B=ScalarParameter() {defaultValue=1, group=FlipBook, parameterName=TileAmount})
  ) {parameterName=SubMovie?}
  n35 = TextureSampleParameter2D(UVs=@n34) {parameterName=FlipBook_01, samplerType=SAMPLERTYPE_LinearColor, texture=&T_VCVFX_Rocksv4_N}
  n36 = TextureSampleParameter2D(UVs=@n34) {parameterName=Flipbook_02, samplerType=SAMPLERTYPE_LinearColor, texture=&T_VCVFX_Rocksv3_N}
  n37 = StaticSwitchParameter(True=LinearInterpolate(A=@n35.RGB, B=@n36.RGB, Alpha=@n3.TextureBlend), False=@n35.RGB) {defaultValue=true, parameterName=Blend}
  n38 = OneMinus(None=Desaturation(None=@n37))
  MaterialExpressionConstantClamp_14 = Saturate(None=@n38)
  n39 = Panner(Coordinate=TextureCoordinate() {uTiling=0.05, vTiling=0.05}) {speedX=0.25, speedY=0.25}
  MaterialExpressionVertexColor_5 = ParticleColor()
  n40 = Add(
   A=Multiply(A=TextureSample(UVs=@n39) {texture=&T_VCVFX_NoiseSpots_RGB}.RGB, B=Constant() {r=100})
   B=@MaterialExpressionVertexColor_5.RGB
  )
  n41 = Constant() {r=1.25}
  MaterialExpressionConstantClamp_15 = Saturate(None=Multiply(A=Power(Base=@n40, Exp=@n41), B=ScalarParameter() {defaultValue=5, parameterName=ReflectionAmount}))
  n42 = Add(
   A=Power(Base=@MaterialExpressionConstantClamp_14, Exp=@n3.ColorPower)
   B=Multiply(A=@MaterialExpressionConstantClamp_15, B=@MaterialExpressionVertexColor_4.RGB)
  )
  MaterialExpressionConstantClamp_16 = Saturate(None=ConstantBiasScale(None=@n38) {bias=-0.2, scale=0.2})
  n43 = Multiply(A=@MaterialExpressionConstantClamp_16, B=@n3.Brightness)
  n44 = Constant3Vector() {constant=<rgba 0 0 1 0>}
  MaterialExpressionConstantClamp_17 = Saturate(None=DotProduct(A=Transform(None=@n37), B=Normalize(VectorInput=@n44)))
  MaterialExpressionConstantClamp_13 = Saturate(None=ConstantBiasScale(None=@MaterialExpressionConstantClamp_17) {bias=0.05})
  MaterialFunctionCall(Emissive=Multiply(A=Constant() {r=0.5}, B=LinearInterpolate(A=@n42, B=@n43, Alpha=@MaterialExpressionConstantClamp_13) {constA=1})) {materialFunction=&MF_FX_EmissiveAdjustments}
  Reroute(None=@n16)
  Reroute(None=@n37)
  n45 = TextureSampleParameter2D() {
   parameterName=FlipBook_01_Alpha
   samplerType=SAMPLERTYPE_LinearColor
   texture=&T_VCVFX_Rocksv4_N
  }
  n46 = TextureSampleParameter2D() {
   parameterName=Flipbook_02_Alpha
   samplerType=SAMPLERTYPE_LinearColor
   texture=&T_VCVFX_Rocksv3_N
  }
  n47 = LinearInterpolate(A=@n45.A, B=@n46.A, Alpha=@n10.TextureBlend)
  Saturate(None=Multiply(A=@n23.A, B=StaticSwitchParameter(True=@n47, False=@n45.A) {defaultValue=true, parameterName=Blend}))
 ]
}
