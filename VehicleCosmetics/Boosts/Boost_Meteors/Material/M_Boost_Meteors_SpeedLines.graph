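# Additive unlit Niagara ribbon material that draws speed lines for the Meteors vehicle boost.
# A tiled, panning spark texture, optionally flattened by desaturation, is shaped by dynamic Power and Brightness and tinted by particle color hue-shifted along U.
# Particle alpha times a squared mask texture forms the opacity term.
# An optional world position offset adds layered sine and cosine waves, faded in from the U=0 end of the ribbon.
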Material /VehicleCosmetics/Boosts/Boost_Meteors/Material/M_Boost_Meteors_SpeedLines
graph {
 properties {
  bAutomaticallySetUsageInEditor=false
  bUsedWithNiagaraRibbons=true
  blendMode=BLEND_Additive
  shadingModel=MSM_Unlit
  usedShadingModels=Unlit
 }
 nodes [
  n1 = ComponentMask(None=TextureCoordinate() {coordinateIndex=1, vTiling=0}) {r=true}
  n2 = ScalarParameter() {defaultValue=1, group=Details, parameterName=TileU}
  n3 = ComponentMask(None=Panner(Coordinate=Multiply(A=@n1, B=@n2), Time=ScalarParameter() {group=Details, parameterName=PanU}) {speedX=0.5}) {r=true}
  n4 = ComponentMask(None=TextureCoordinate() {coordinateIndex=1, uTiling=0}) {g=true}
  n5 = ScalarParameter() {defaultValue=1, group=Details, parameterName=TileV}
  n6 = ComponentMask(None=Panner(Coordinate=Multiply(A=@n4, B=@n5), Time=ScalarParameter() {group=Details, parameterName=PanV}) {speedY=0.5}) {g=true}
  n7 = Time()
  n8 = DynamicParameter() {paramNames [PanSpeed,Brightness,Power,Param4]}
  n9 = TextureSampleParameter2D(UVs=Panner(Coordinate=AppendVector(A=@n3, B=@n6), Time=Multiply(A=@n7, B=@n8.PanSpeed)) {speedX=1}) {parameterName=RGB_Texture, texture=/VehicleCosmetics/Boosts/Boost_Meteors/Texture/T_VCVFX_Boost_Meteors_TrailSparks_RGB}
  n10 = Power(
   Base=StaticSwitchParameter(True=Desaturation(None=@n9.RGB), False=StaticComponentMaskParameter(None=@n9.RGB) {defaultR=true, parameterName=TextureRGB}) {parameterName=FlattenTexture}
   Exp=@n8.Power
  )
  n11 = TextureCoordinate()
  n12 = Subtract(A=OneMinus(None=ComponentMask(None=@n11) {r=true}), B=ScalarParameter() {defaultValue=0.6, group=Color, parameterName=HueShift_Amount})
  n13 = ScalarParameter() {defaultValue=0.3, group=Color, parameterName=HueShift_Sharpness}
  MaterialExpressionVertexColor_3 = ParticleColor()
  MaterialExpressionConstantClamp_11 = Saturate(None=MaterialFunctionCall(Hue Shift Percentage=Multiply(A=@n12, B=@n13), Texture=@MaterialExpressionVertexColor_3.RGB) {materialFunction=&HueShift}.Result)
  MaterialFunctionCall(Emissive=Multiply(A=Multiply(A=@n10, B=@MaterialExpressionConstantClamp_11), B=@n8.Brightness)) {materialFunction=&MF_FX_EmissiveAdjustments}
  n14 = TextureSample() {texture=/VehicleCosmetics/Boosts/Boost_Meteors/Texture/T_VCVFX_Boost_Meteors_Mask_RGB}
  Multiply(A=@MaterialExpressionVertexColor_3.A, B=Multiply(A=@n14.R, B=@n14.R))
  n15 = TextureCoordinate()
  MaterialExpressionConstantClamp_10 = Saturate(None=Multiply(A=ComponentMask(None=@n15) {r=true}, B=Constant() {r=8}))
  n16 = Multiply(A=ScalarParameter() {group=WPO, parameterName=WaveSpeed}, B=OneMinus(None=Time()))
  n17 = ComponentMask(None=Multiply(A=TextureCoordinate() {coordinateIndex=1, uTiling=2}, B=ScalarParameter() {group=WPO, parameterName=tile})) {r=true}
  n18 = Add(A=@n16, B=@n17)
  n19 = Add(A=Cosine(None=@n18), B=Cosine(None=@n18) {period=2})
  n20 = Add(A=@n17, B=@n16)
  n21 = Add(A=Sine(None=@n20), B=Sine(None=@n20) {period=2})
  n22 = Transform(None=AppendVector(A=AppendVector(A=@n19, B=@n21), B=@n19))
  n23 = ScalarParameter() {group=WPO, parameterName=Intensity}
  StaticSwitchParameter(True=Multiply(A=@MaterialExpressionConstantClamp_10, B=Multiply(A=@n22, B=@n23)), False=Constant()) {defaultValue=true, group=WPO, parameterName=WPO}
 ]
}
