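# Translucent unlit Niagara ribbon material for the streaks of the Meteors vehicle boost.
# Two optional flipbook textures are channel-masked or desaturated, then blended by multiply, add or dynamic-parameter lerp, with optional panning mask modulation.
# The result, faded along ribbon U, drives emissive and opacity, with particle color and alpha rebalanced by brightness.
# Static switches also enable depth fade and summed sine/cosine world position offset waves.
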Material /VehicleCosmetics/Boosts/Boost_Meteors/Material/M_Boost_Meteors_Streaks
graph {
 properties {
  bAutomaticallySetUsageInEditor=false
  bUsedWithNiagaraRibbons=true
  blendMode=BLEND_Translucent
  shadingModel=MSM_Unlit
  usedShadingModels=Unlit
 }
 nodes [
  n1 = Constant() {r=0.1}
  MaterialExpressionVertexColor_0 = ParticleColor()
  n2 = Add(A=@MaterialExpressionVertexColor_0.RGB, B=Constant() {r=0.05})
  n3 = Multiply(A=@n2, B=Constant() {r=1.2})
  n4 = Add(A=@MaterialExpressionVertexColor_0.R, B=@MaterialExpressionVertexColor_0.R)
  MaterialExpressionConstantClamp_2 = Saturate(None=Divide(A=Add(A=@n4, B=@MaterialExpressionVertexColor_0.B), B=Constant() {r=3}))
  n5 = Constant() {r=0.78}
  n6 = Add(A=@MaterialExpressionVertexColor_0.RGB, B=Constant() {r=0.05})
  n7 = Saturate(None=LinearInterpolate(
    A=If(A=@MaterialExpressionConstantClamp_2, B=@n1, "A > B"=@MaterialExpressionVertexColor_0.RGB, "A == B"=@n2, "A < B"=@n3)
    B=If(
     A=@MaterialExpressionConstantClamp_2
     B=@n5
     "A > B"=@MaterialExpressionVertexColor_0.RGB
     "A == B"=@MaterialExpressionVertexColor_0.RGB
     "A < B"=@n6
    )
    Alpha=@MaterialExpressionVertexColor_0.RGB
   ))
  n8 = ScalarParameter() {defaultValue=1, group=Pan, parameterName=PanTile}
  n9 = Multiply(A=TextureCoordinate() {uTiling=0.05, vTiling=0.25}, B=@n8)
  n10 = Multiply(A=Time(), B=ScalarParameter() {defaultValue=1, group=Pan, parameterName=PanSpeed})
  n11 = TextureCoordinate() {uTiling=0.25, vTiling=0.05}
  n12 = StaticSwitchParameter(
   True=Rotator(Coordinate=Panner(Coordinate=@n9, Time=@n10) {speedY=0.4}, Time=Constant() {r=0.5}) {speed=3.14}
   False=Panner(Coordinate=Multiply(A=@n11, B=@n8), Time=@n10) {speedX=0.4}
  ) {group=FlipBook, parameterName=RotateFlipbook}
  n13 = Multiply(
   A=ComponentMask(None=TextureCoordinate() {coordinateIndex=1, vTiling=0}) {r=true}
   B=ScalarParameter() {defaultValue=1, group=Details, parameterName=TileU}
  )
  n14 = Multiply(
   A=ComponentMask(None=TextureCoordinate() {coordinateIndex=1, uTiling=0}) {g=true}
   B=ScalarParameter() {defaultValue=1, group=Details, parameterName=TileV}
  )
  n15 = ScalarParameter() {group=Details, parameterName=PanV}
  n16 = AppendVector(
   A=ComponentMask(None=Panner(Coordinate=@n13) {speedX=-0.5}) {r=true}
   B=ComponentMask(None=Panner(Coordinate=@n14, Time=@n15) {speedY=0.5}) {g=true}
  )
  n17 = Constant() {r=0.5}
  n18 = ScalarParameter() {defaultValue=4, group=FlipBook, parameterName=Rows}
  n19 = ScalarParameter() {defaultValue=4, group=FlipBook, parameterName=Columns}
  n20 = AppendVector(A=@n19, B=@n18)
  n21 = Divide(
   A=StaticSwitchParameter(True=Rotator(Coordinate=@n16, Time=@n17) {speed=3.14}, False=@n16) {group=FlipBook, parameterName=RotateFlipbook}
   B=@n20
  )
  n22 = Frac(None=Multiply(A=Time(), B=ScalarParameter() {group=FlipBook, parameterName=Speed}))
  n23 = AppendVector(A=@n22, B=@n22)
  n24 = AppendVector(A=Multiply(A=@n19, B=@n18), B=@n18)
  n25 = Divide(A=Floor(None=Multiply(A=@n24, B=@n23)), B=@n20)
  n26 = Rotator(Coordinate=@n16, Time=Constant() {r=0.5}) {speed=3.14}
  n27 = TextureSampleParameter2D(UVs=StaticSwitchParameter(True=Add(A=@n21, B=@n25), False=StaticSwitchParameter(True=@n26, False=@n16) {group=FlipBook, parameterName=RotateFlipbook}) {defaultValue=true, group=FlipBook, parameterName=FlipBook}) {group=FlipBook, parameterName=BaseTexture, texture=/DealershipSharedLibrary/FX/Textures/Gradient_Masks/T_VCVFX_BasicTrail_A}
  n28 = Multiply(
   A=ComponentMask(None=TextureCoordinate() {coordinateIndex=1, vTiling=0}) {r=true}
   B=ScalarParameter() {defaultValue=1, group=Details, parameterName=2_TileU}
  )
  n29 = Multiply(
   A=ComponentMask(None=TextureCoordinate() {coordinateIndex=1, uTiling=0}) {g=true}
   B=ScalarParameter() {defaultValue=1, group=Details, parameterName=2_TileV}
  )
  n30 = ScalarParameter() {group=Details, parameterName=2_PanV}
  n31 = AppendVector(
   A=ComponentMask(None=Panner(Coordinate=@n28) {speedX=-0.5}) {r=true}
   B=ComponentMask(None=Panner(Coordinate=@n29, Time=@n30) {speedY=0.5}) {g=true}
  )
  n32 = Rotator(Coordinate=@n31, Time=Constant() {r=0.5}) {speed=3.14}
  n33 = ScalarParameter() {defaultValue=4, group=FlipBook, parameterName=Rows_02}
  n34 = ScalarParameter() {defaultValue=4, group=FlipBook, parameterName=Columns_02}
  n35 = AppendVector(A=@n34, B=@n33)
  n36 = Frac(None=Multiply(A=Time(), B=ScalarParameter() {defaultValue=1, group=FlipBook, parameterName=Speed_02}))
  n37 = Multiply(A=AppendVector(A=Multiply(A=@n34, B=@n33), B=@n33), B=AppendVector(A=@n36, B=@n36))
  n38 = Add(
   A=Divide(A=StaticSwitchParameter(True=@n32, False=@n31) {group=FlipBook, parameterName=RotateFlipbook_02}, B=@n35)
   B=Divide(A=Floor(None=@n37), B=@n35)
  )
  n39 = StaticSwitchParameter(True=Rotator(Coordinate=@n31, Time=Constant() {r=0.5}) {speed=3.14}, False=@n31) {group=FlipBook, parameterName=RotateFlipbook_02}
  n40 = StaticSwitchParameter(
   True=@n27.RGB
   False=TextureSampleParameter2D(UVs=StaticSwitchParameter(True=@n38, False=@n39) {defaultValue=true, group=FlipBook, parameterName=Flipbook_02}) {group=FlipBook, parameterName=BaseTexture_02, texture=&T_VCVFX_BasicTrail_A}.RGB
  ) {defaultValue=true, group=Blend, parameterName=SingleTexture}
  n41 = Multiply(
   A=ScalarParameter() {defaultValue=1, group=Blend, parameterName=Blend_02_Brightness}
   B=StaticSwitchParameter(True=StaticComponentMaskParameter(None=@n40) {defaultB=true, group=Blend, parameterName=Blend_02}, False=Desaturation(None=@n40)) {defaultValue=true, group=Blend, parameterName=Use_RGB_02}
  )
  n42 = DynamicParameter() {paramNames [Brightness,LerpBlend,Saturation,ColorShift]}
  n43 = Multiply(
   A=ScalarParameter() {defaultValue=1, group=Blend, parameterName=Blend_01_Brightness}
   B=StaticSwitchParameter(True=StaticComponentMaskParameter(None=@n27.RGB) {defaultG=true, group=Blend, parameterName=Blend_01}, False=Desaturation(None=@n27.RGB)) {defaultValue=true, group=Blend, parameterName=Use_RGB}
  )
  n44 = Multiply(A=@n43, B=@n41)
  n45 = Add(A=@n43, B=@n41)
  n46 = StaticSwitchParameter(
   True=StaticSwitchParameter(
    True=LinearInterpolate(A=@n43, B=@n41, Alpha=@n42.LerpBlend)
    False=StaticSwitchParameter(True=@n44, False=@n45) {group=Blend, parameterName="BlendType(T=M)(F=A)"}
   ) {group=Blend, parameterName=LerpBlend}
   False=Multiply(
    A=StaticComponentMaskParameter(None=@n27.RGB) {defaultG=true, group=Blend, parameterName=RGB_Mask}
    B=ScalarParameter() {defaultValue=1, group=Blend, parameterName=RGB_Mask_Brightness}
   )
  ) {group=Blend, parameterName=Blend}
  n47 = Multiply(
   A=Multiply(
    A=ScalarParameter() {defaultValue=3, group=Pan, parameterName=PanBrightness}
    B=TextureSample(UVs=@n12) {texture=/DealershipSharedLibrary/FX/Textures/Gradient_Masks/T_VCVFX_TimeMask_A}.RGB
   )
   B=@n46
  )
  n48 = StaticSwitchParameter(True=Add(A=@n47, B=@n46), False=@n47) {group=Pan, parameterName=AddPan?}
  n49 = TextureCoordinate()
  n50 = Frac(None=TextureCoordinate() {uTiling=4, vTiling=4})
  n51 = ComponentMask(None=StaticSwitchParameter(True=@n49, False=StaticSwitchParameter(True=@n49, False=@n50) {defaultValue=true, group=FlipBook, parameterName=NotAnAtlas}) {defaultValue=true, group=FlipBook, parameterName=FlipBook}) {r=true}
  MaterialExpressionConstantClamp_5 = Saturate(None=Multiply(A=@n51, B=Constant() {r=30}))
  n52 = OneMinus(None=@n51)
  n53 = Multiply(
   A=Multiply(
    A=StaticSwitchParameter(True=@n48, False=@n46) {group=Pan, parameterName=PanLightning}
    B=StaticSwitchParameter(True=@n51, False=@MaterialExpressionConstantClamp_5) {defaultValue=true, parameterName=FadeIn?}
   )
   B=@n52
  )
  MaterialExpressionConstantClamp_0 = Saturate(None=@n52)
  n54 = Multiply(A=Saturate(None=@n53), B=Multiply(A=@MaterialExpressionConstantClamp_0, B=Constant() {r=8}))
  n55 = ParticleColor()
  MaterialFunctionCall(Emissive=Multiply(A=Multiply(A=@n7, B=@n54), B=@n55.RGB)) {materialFunction=&MF_FX_EmissiveAdjustments}
  MaterialExpressionVertexColor_1 = ParticleColor()
  n56 = Multiply(A=@MaterialExpressionVertexColor_1.A, B=Constant() {r=3})
  n57 = If(
   A=@MaterialExpressionConstantClamp_2
   B=Constant() {r=0.1}
   "A > B"=Multiply(A=Constant() {r=1.5}, B=@MaterialExpressionVertexColor_1.A)
   "A == B"=@n56
   "A < B"=@n56
  )
  n58 = If(
   A=@MaterialExpressionConstantClamp_2
   B=Constant() {r=0.78}
   "A > B"=@MaterialExpressionVertexColor_1.A
   "A == B"=@MaterialExpressionVertexColor_1.A
   "A < B"=Multiply(A=@MaterialExpressionVertexColor_1.A, B=Constant() {r=1.2})
  )
  n59 = Multiply(A=Saturate(None=LinearInterpolate(A=@n57, B=@n58, Alpha=@MaterialExpressionConstantClamp_2)), B=@n53)
  n60 = MaterialFunctionCall(Opacity In=Constant() {r=1}, DepthFade Distance=Constant() {r=100}) {materialFunction=&MF_VCVFX_DepthFade}
  MaterialExpressionConstantClamp_6 = Saturate(None=StaticSwitchParameter(True=Multiply(A=@n59, B=@n60.Result), False=@n59) {group=Opacity, parameterName=DBA?})
  Multiply(A=@n55.A, B=@MaterialExpressionConstantClamp_6)
  ScalarParameter() {group=Details, parameterName=2_PanU}
  ScalarParameter() {group=Details, parameterName=PanU}
  n61 = TextureCoordinate()
  MaterialExpressionConstantClamp_7 = Saturate(None=Multiply(A=ComponentMask(None=@n61) {r=true}, B=Constant() {r=8}))
  n62 = Multiply(A=ScalarParameter() {group=WPO, parameterName=WaveSpeed}, B=OneMinus(None=Time()))
  n63 = ComponentMask(None=Multiply(A=TextureCoordinate() {coordinateIndex=1, uTiling=2}, B=ScalarParameter() {group=WPO, parameterName=tile})) {r=true}
  n64 = Add(A=@n62, B=@n63)
  n65 = Add(A=Cosine(None=@n64), B=Cosine(None=@n64) {period=2})
  n66 = Add(A=@n63, B=@n62)
  n67 = Add(A=Sine(None=@n66), B=Sine(None=@n66) {period=2})
  n68 = Transform(None=AppendVector(A=AppendVector(A=@n65, B=@n67), B=@n65))
  n69 = ScalarParameter() {group=WPO, parameterName=Intensity}
  StaticSwitchParameter(True=Multiply(A=@MaterialExpressionConstantClamp_7, B=Multiply(A=@n68, B=@n69)), False=Constant()) {defaultValue=true, group=WPO, parameterName=WPO}
 ]
}
